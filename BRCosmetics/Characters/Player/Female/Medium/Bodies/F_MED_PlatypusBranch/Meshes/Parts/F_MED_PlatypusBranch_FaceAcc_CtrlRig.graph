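# Control Rig for the F_MED_PlatypusBranch FaceAcc part that applies corrective bone offsets, including to the braid master bone, as the head moves.
# A SphericalPoseReader measures the head relative to spine_05 within an asymmetric falloff region, with debug drawing on.
# Its output weights a ModifyTransforms node that applies the offsets.
# A Debug-gated Controls_1 function copies test controls onto spine, neck, head and clavicle bones, but is not wired into execution.
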ControlRigBlueprint /BRCosmetics/Characters/Player/Female/Medium/Bodies/F_MED_PlatypusBranch/Meshes/Parts/F_MED_PlatypusBranch_FaceAcc_CtrlRig
graphs [
 New Function_ContainedGraph {nodes [
   n1 = RigVMFunctionEntryNode()
   RigVMFunction_ControlFlowBranch = &RigVMFunction_ControlFlowBranch(Condition=@n1.Debug, ExecuteContext=@n1.ExecuteContext)
   RigVMFunctionReturnNode(ExecuteContext=@RigVMFunction_ControlFlowBranch.Completed)
   RigUnit_GetTransform = &RigUnit_GetTransform(Item=(Type=Control,Name="spine_05_ctrl"), bInitial=False)
   RigUnit_SetTransform = &RigUnit_SetTransform(
    Item=(Type=Bone,Name="spine_05")
    Weight=1
    bInitial=False
    Value=@RigUnit_GetTransform.Transform
    ExecutePin=@RigVMFunction_ControlFlowBranch.True
   )
   > &RigUnit_SetTransform(
    Item=(Type=Bone,Name="neck_01")
    Weight=1
    bInitial=False
    Value=&RigUnit_GetTransform(Item=(Type=Control,Name="neck_01_ctrl"), bInitial=False).Transform
   )
   > &RigUnit_SetTransform(
    Item=(Type=Bone,Name="neck_02")
    Weight=1
    bInitial=False
    Value=&RigUnit_GetTransform(Item=(Type=Control,Name="neck_02_ctrl"), bInitial=False).Transform
   )
   > &RigUnit_SetTransform(
    Item=(Type=Bone,Name="head")
    Weight=1
    bInitial=False
    Value=&RigUnit_GetTransform(Item=(Type=Control,Name="head_ctrl"), bInitial=False).Transform
   )
   > &RigUnit_SetTransform(
    Item=(Type=Bone,Name="clavicle_l")
    Weight=1
    bInitial=False
    Value=&RigUnit_GetTransform(Item=(Type=Control,Name="clavicle_l_ctrl"), bInitial=False).Transform
   )
   > &RigUnit_SetTransform(
    Item=(Type=Bone,Name="clavicle_r")
    Weight=1
    bInitial=False
    Value=&RigUnit_GetTransform(Item=(Type=Control,Name="clavicle_r_ctrl"), bInitial=False).Transform
   )
  ]}
 RigVMFunctionLibrary {nodes [Controls_1 = RigVMCollapseNode(Debug=False, ExecuteContext=())]}
 RigVMModel {nodes [
   RigUnit_BeginExecution = &RigUnit_BeginExecution()
   n1 = &RigUnit_SphericalPoseReader(
    ActiveRegionScaleFactors=(PositiveWidth=3.2,NegativeWidth=1,PositiveHeight=1,NegativeHeight=1)
    ActiveRegionSize=0.1
    Debug=(bDrawDebug=true,bDraw2D=False,bDrawLocalAxes=False,DebugScale=25,DebugSegments=20,DebugThickness=0.25)
    DriverAxis=(X=1,Y=0,Z=0)
    DriverItem=(Type=Bone,Name="head")
    FalloffRegionScaleFactors=(PositiveWidth=1.4,NegativeWidth=1,PositiveHeight=1,NegativeHeight=1)
    FalloffSize=0.4
    FlipHeightScaling=False
    FlipWidthScaling=False
    OptionalParentItem=(Type=Bone,Name="spine_05")
    RotationOffset=(X=-20,Y=0,Z=0)
    ExecutePin=@RigUnit_BeginExecution.ExecutePin
   )
   > &RigUnit_ModifyTransforms(
    ItemToModify="((Item=(Type=Bone,Name=\"None\"),Transform=(Rotation=(X=0,Y=0,Z=0,W=1),Translation=(X=0,Y=0,Z=0),Scale3D=(X=1,Y=1,Z=1))),(Item=(Type=Bone,Name=\"braid_MSTR\"),Transform=(Rotation=(X=0,Y=0,Z=0,W=1),Translation=(X=-3,Y=0,Z=-1),Scale3D=(X=1,Y=1,Z=1))),(Item=(Type ...(truncated)"
    WeightMinimum=0
    Weight=@n1.OutputParam
   )
   Controls_1 = RigVMFunctionReferenceNode(Debug=true, ExecuteContext=())
  ]}
]
hierarchy [
 root {loc=<0 0 0>, rot=<0 0 0 1>}
 pelvis {parent=root, loc=<0 -0.5301 98.0571>, rot=<0.0127 0.707 -0.0127 -0.707>}
 spine_01 {parent=pelvis, loc=<6.1208 -0.9328 0>, rot=<0 0 0.034 -0.9994>}
 spine_02 {parent=spine_01, loc=<6.5997 -0.2095 0>, rot=<0 0 0.0238 -0.9997>}
 spine_03 {parent=spine_02, loc=<6.4914 -0.1237 0>, rot=<0 0 0.0149 -0.9999>}
 spine_04 {parent=spine_03, loc=<6.9598 0.0794 0>, rot=<0 0 -0.0132 -0.9999>}
 spine_05 {parent=spine_04, loc=<6.9987 0 0>, rot=<0 0 -0.0015 -1>}
 neck_01 {parent=spine_05, loc=<11.3138 0.1529 0>, rot=<0 0 -0.2003 -0.9797>}
 neck_02 {parent=neck_01, loc=<4.5627 0 0>, rot=<0 0 0.0504 -0.9987>}
 head {parent=neck_02, loc=<5.3914 0 0>, rot=<0 0 0.115 -0.9934>}
 base_hair_r {parent=head, loc=<6.694 -2.228 8.4306>, rot=<0 0 1 0.0059>}
 dyn_hair_r {parent=base_hair_r, loc=<0 0 0>, rot=<0 0 0 -1>}
 base_hair_l {parent=head, loc=<6.6794 -2.2278 -8.4306>, rot=<-0.0059 -1 0 0>}
 dyn_hair_l {parent=base_hair_l, loc=<0 0 0>, rot=<0 0 0 1>}
 base_hair_frnt_l {parent=head, loc=<7.2965 4.6733 -6.5377>, rot=<0.0137 -0.9865 -0.1607 -0.0267>}
 dyn_hair_frnt_l {parent=base_hair_frnt_l, loc=<0 0 0>, rot=<0 0 0 1>}
 base_hair_top_r {parent=head, loc=<10.3469 7.4689 -3.8705>, rot=<0.3799 0.6379 0.6508 -0.1588>}
 dyn_hair_top_r {parent=base_hair_top_r, loc=<0 0 0>, rot=<0 0 0 -1>}
 base_hair_top {parent=head, loc=<14.6786 1.3312 0.2037>, rot=<0.6315 -0.0168 0.7712 -0.0783>}
 dyn_hair_top {parent=base_hair_top, loc=<0 0 0>, rot=<0 0 0 -1>}
 base_hair_frnt_r {parent=head, loc=<5.3959 7.2904 6.7369>, rot=<0.2006 0.3159 0.9244 -0.0736>}
 dyn_hair_frnt_r {parent=base_hair_frnt_r, loc=<0 0 0>, rot=<0 0 0 -1>}
 base_hair_back {parent=head, loc=<5.5143 -9.446 0>, rot=<0.0052 -0.7071 0.7071 -0.0052>}
 dyn_hair_back {parent=base_hair_back, loc=<0 0 0>, rot=<0 0 0 1>}
 braid_MSTR {parent=head, loc=<-3.0526 -6.0537 2.4741>, rot=<-0.069 -0.8787 0.3863 0.2717>}
 base_braid {parent=braid_MSTR, loc=<4.7134 0 0>, rot=<-0.7549 -0.2243 0.02 -0.616>}
 dyn_braid_00 {parent=base_braid, loc=<0 0 0>, rot=<0 0 0 1>}
 dyn_braid_01 {parent=dyn_braid_00, loc=<3.7129 -0.0637 -0.0516>, rot=<0.0546 -0.2677 0.0207 -0.9617>}
 dyn_braid_02 {parent=dyn_braid_01, loc=<3.5765 0 0>, rot=<-0.0075 0.0828 -0.1242 -0.9888>}
 dyn_braid_03 {parent=dyn_braid_02, loc=<3.5765 0 0>, rot=<-0.1075 0.1254 -0.0544 -0.9848>}
 main_hood_on {parent=neck_02, loc=<-4.5395 -0.4596 0>, rot=<0.4418 -0.5521 0.5521 -0.4418>}
 base_hood_on_bk {parent=main_hood_on, loc=<6.7458 0 0.1383>, rot=<0 -0.5762 0 -0.8173>}
 dyn_hood_on_bk {parent=base_hood_on_bk, loc=<0 0 0>, rot=<0 0 0 1>}
 clavicle_l {parent=spine_05, loc=<5.4831 -0.1053 -3.2946>, rot=<0.0465 -0.7289 -0.009 -0.6829>}
 upperarm_l {parent=clavicle_l, loc=<9.8658 -1.2136 -0.4105>, rot=<0.0164 -0.3245 -0.0269 -0.9454>}
 upperarm_twist_01_l {parent=upperarm_l, loc=<7.2135 -0.1134 0>, rot=<0 0 0 1>}
 deform_upperarm_bk_base_l {parent=upperarm_twist_01_l, loc=<-6.4053 -4.3746 0.314>, rot=<0 0 0 -1>}
 deform_upperarm_bk_l {parent=deform_upperarm_bk_base_l, loc=<0 0 0>, rot=<0 0 0 1>}
 deform_upperarm_up_base_l {parent=upperarm_twist_01_l, loc=<-8.891 0.4825 3.9517>, rot=<0 0 0 -1>}
 deform_upperarm_up_l {parent=deform_upperarm_up_base_l, loc=<0 0 0>, rot=<0 0 0 1>}
 base_hood_on_l {parent=clavicle_l, loc=<7.0361 0.658 7.6076>, rot=<0.0008 0.0336 -0.0252 -0.9991>}
 dyn_hood_on_l {parent=base_hood_on_l, loc=<0 0 0>, rot=<0 0 0 1>}
 clavicle_r {parent=spine_05, loc=<5.4831 -0.1053 3.2946>, rot=<0.7289 0.0465 -0.6829 0.009>}
 upperarm_r {parent=clavicle_r, loc=<-9.8658 1.2136 0.4105>, rot=<0.0164 -0.3245 -0.0269 -0.9454>}
 upperarm_twist_01_r {parent=upperarm_r, loc=<-7.2135 0.1134 0>, rot=<0 0 0 -1>}
 deform_upperarm_bk_base_r {parent=upperarm_twist_01_r, loc=<6.4053 4.3746 -0.314>, rot=<0 0 0 -1>}
 deform_upperarm_bk_r {parent=deform_upperarm_bk_base_r, loc=<0 0 0>, rot=<0 0 0 1>}
 deform_upperarm_up_base_r {parent=upperarm_twist_01_r, loc=<8.891 -0.4825 -3.9517>, rot=<0 0 0 -1>}
 deform_upperarm_up_r {parent=deform_upperarm_up_base_r, loc=<0 0 0>, rot=<0 0 0 1>}
 base_hood_on_r {parent=clavicle_r, loc=<-7.0361 -0.658 -7.6072>, rot=<0.0252 -0.9991 0.0008 -0.0336>}
 dyn_hood_on_r {parent=base_hood_on_r, loc=<0 0 0>, rot=<0 0 0 1>}
 attach_backpack {parent=spine_05, loc=<-0.123 -13.473 0>, rot=<-0.0093 -0.707 0.0093 -0.707>}
 attach_cape {parent=spine_05, loc=<9.017 -11.714 0>, rot=<0.0616 -0.7044 -0.0616 -0.7044>}
 deform_pec_aim_base_l {parent=spine_04, loc=<3.2456 9.3309 -1.2237>, rot=<0.0603 -0.4033 0.3318 -0.8507>}
 deform_pec_aim_l {parent=deform_pec_aim_base_l, loc=<0 0 0>, rot=<0 0 0 1>}
 deform_pec_base_l {parent=deform_pec_aim_l, loc=<9.3665 0 0>, rot=<0 0 0 1>}
 deform_pec_l {parent=deform_pec_base_l, loc=<0 0 0>, rot=<0 0 0 1>}
 deform_pec_aim_base_r {parent=spine_04, loc=<3.2452 9.3309 1.2237>, rot=<-0.4033 -0.0603 0.8507 0.3318>}
 deform_pec_aim_r {parent=deform_pec_aim_base_r, loc=<0 0 0>, rot=<0 0 0 1>}
 deform_pec_base_r {parent=deform_pec_aim_r, loc=<-9.3667 -0.0003 -0.0004>, rot=<0 0 0 1>}
 deform_pec_r {parent=deform_pec_base_r, loc=<0 0 0>, rot=<0 0 0 1>}
 spine_05_ctrl {controlType=EULER_TRANSFORM, type=Control, loc=<0 0 0>, rot=<0 0 0 1>}
 neck_01_ctrl {controlType=EULER_TRANSFORM, parent=spine_05_ctrl, type=Control, loc=<0 0 0>, rot=<0 0 0 1>}
 neck_02_ctrl {controlType=EULER_TRANSFORM, parent=neck_01_ctrl, type=Control, loc=<0 0 0>, rot=<0 0 0 1>}
 head_ctrl {controlType=EULER_TRANSFORM, parent=neck_02_ctrl, type=Control, loc=<0 0 0>, rot=<0 0 0 1>}
 clavicle_r_ctrl {controlType=EULER_TRANSFORM, parent=spine_05_ctrl, type=Control, loc=<0 0 0>, rot=<0 0 0 1>}
 clavicle_l_ctrl {controlType=EULER_TRANSFORM, parent=spine_05_ctrl, type=Control, loc=<0 0 0>, rot=<0 0 0 1>}
]
