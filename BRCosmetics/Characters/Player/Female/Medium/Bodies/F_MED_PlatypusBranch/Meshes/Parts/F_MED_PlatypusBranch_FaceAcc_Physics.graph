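# Physics asset for the face accessory part of the PlatypusBranch female medium character body, covering a dynamic braid and nearby upper-body collision.
# A chain of small capsules runs from base_braid through four dyn_braid bones, linked by constraints with swing limits and tight twist limits.
# Larger capsules on head, neck, upper spine and right clavicle form body volumes around the braid, mostly joined by free angular constraints.
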PhysicsAsset /BRCosmetics/Characters/Player/Female/Medium/Bodies/F_MED_PlatypusBranch/Meshes/Parts/F_MED_PlatypusBranch_FaceAcc_Physics
bodies {
 base_braid [{shapeType=Capsule, length=0.505, radius=0.505, shapeName=base_braid_capsule, rotation=<rot 0 -90 0>}]
 clavicle_r [{
   shapeType=Capsule
   length=7.3954
   radius=4.3436
   shapeName=clavicle_r_capsule
   center=<-8.1354 -2.8893 -3.8582>
   rotation=<rot 70.0002 95.0002 -10>
  }]
 dyn_braid_00 [{shapeType=Capsule, length=3.0455, radius=1, shapeName=dyn_braid_00_capsule, center=<0.8911 0.0643 0.2215>, rotation=<rot 90 0 0>}]
 dyn_braid_01 [{
   shapeType=Capsule
   length=2.6611
   radius=0.9
   shapeName=dyn_braid_01_capsule
   center=<1.1024 -0.4448 -0.058>
   rotation=<rot 0 -90.1996 -89.3572>
  }]
 dyn_braid_02 [{
   shapeType=Capsule
   length=1.4949
   radius=0.85
   shapeName=dyn_braid_02_capsule
   center=<1.5151 -0.0354 -0.1096>
   rotation=<rot 0 -94.2435 -86.4635>
  }]
 dyn_braid_03 [{
   shapeType=Capsule
   length=2.9239
   radius=0.8346
   shapeName=dyn_braid_03_capsule
   center=<1.6946 0.0192 -0.0775>
   rotation=<rot 0 -86.7722 -88.9112>
  }]
 head [{shapeType=Capsule, length=4.9908, radius=10.5269, center=<4.3258 0 0>, rotation=<rot 90 0 0>}]
 neck_01 [{shapeType=Capsule, length=2, radius=4.8, shapeName=neck_01_capsule, rotation=<rot 90 0 0>}]
 spine_04 [{shapeType=Capsule, length=11.8404, radius=9.0629, center=<3.1739 3.0251 0.0178>, rotation=<rot 0.0657 -90 -0.0254>}]
 spine_05 [
  {shapeType=Capsule, length=11.8404, radius=9.0629, center=<1.6186 -1.813 0.0227>}
  {
   shapeType=Capsule
   length=11
   radius=3.5
   shapeName=spine_05_capsule
   center=<10.4016 1.4813 10.2504>
   rotation=<rot 5.7162 34.5863 106.7811>
  }
  {shapeType=Capsule, length=15, radius=3.5, shapeName=spine_05_capsule, center=<9.691 -8.8041 0>}
 ]
}
constraints [
 {bone1Name=head, bone2Name=neck_01, swing1LimitDegrees=45, swing2LimitDegrees=45, twistLimitDegrees=45}
 {
  bone1Name=base_braid
  bone2Name=head
  swing1LimitDegrees=45
  swing1Motion=Free
  swing2LimitDegrees=45
  swing2Motion=Free
  twistLimitDegrees=45
  twistMotion=Free
 }
 {bone1Name=dyn_braid_01, bone2Name=dyn_braid_00, swing1LimitDegrees=60, swing2LimitDegrees=45, twistLimitDegrees=5}
 {bone1Name=dyn_braid_02, bone2Name=dyn_braid_01, swing1LimitDegrees=60, swing2LimitDegrees=45, twistLimitDegrees=5}
 {bone1Name=dyn_braid_03, bone2Name=dyn_braid_02, swing1LimitDegrees=45, swing2LimitDegrees=45, twistLimitDegrees=5}
 {
  bone1Name=neck_01
  bone2Name=spine_05
  swing1LimitDegrees=45
  swing1Motion=Free
  swing2LimitDegrees=45
  swing2Motion=Free
  twistLimitDegrees=45
  twistMotion=Free
 }
 {bone1Name=dyn_braid_00, bone2Name=base_braid, swing1LimitDegrees=60, swing2LimitDegrees=45, twistLimitDegrees=5}
 {
  bone1Name=spine_05
  bone2Name=spine_04
  swing1LimitDegrees=45
  swing1Motion=Free
  swing2LimitDegrees=45
  swing2Motion=Free
  twistLimitDegrees=45
  twistMotion=Free
 }
 {
  bone1Name=clavicle_r
  bone2Name=spine_05
  swing1LimitDegrees=45
  swing1Motion=Free
  swing2LimitDegrees=45
  swing2Motion=Free
  twistLimitDegrees=45
  twistMotion=Free
 }
]
