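# Post-process style AnimBlueprint for the F_MED_PlatypusBranchHood face accessory, animating hood, bangs and braid on the pose copied from the attached parent.
# A PlatypusBranch Control Rig feeds AnimDynamics for hood and bangs (off on ziplines), then RigidBody physics with state-driven gravity and sim-space settings.
# Additive offsets follow targeting and head pitch.
# Weighted constraints pull the braid toward spine, neck, head and clavicles during front end, emotes or an LOD match.
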AnimBlueprint /BRCosmetics/Characters/Player/Female/Medium/Bodies/F_MED_PlatypusBranchHood/Meshes/Parts/F_MED_PlatypusBranchHood_FaceAcc_AnimBP
target_skeleton=&/F_MED_PlatypusBranchHood_FaceAcc_Skeleton
graphs [
 AnimGraph {nodes [
   n1 = Control Rig(Source=Copy Pose From Mesh() {bUseAttachedParent=true}) {
    alphaBoolBlend {blendOption=Linear}
    controlRigAssetReference {blueprintRigClass=/BRCosmetics/Characters/Player/Female/Medium/Bodies/F_MED_PlatypusBranch/Meshes/Parts/F_MED_PlatypusBranch_FaceAcc_CtrlRig.F_MED_PlatypusBranch_FaceAcc_CtrlRig_C}
   }
   n2 = "Constraint - base_bang_2 (PPPP)"!(
    ComponentPose=Local To Component(LocalPose=@n1)
    Alpha=1
    ConstraintWeights_0=0.9609
    ConstraintWeights_1:float=0.0378
    ConstraintWeights_2:float=0.0009
    ConstraintWeights_3:float=0.0005
   ) -> (Pose) {
    alphaBoolBlend {blendOption=Linear}
    boneToModify {boneName=base_bang_2}
    constraintSetup [{targetBone {boneName=head}}, {targetBone {boneName=dyn_hood_on_r}}, {targetBone {boneName=clavicle_r}}, {targetBone {boneName=spine_05}}]
    constraintWeights [1,1,1,1]
   }
   n3 = Get animD alpha() -> (animD alpha:double)
   n4 = Anim Dynamics - Bone: dyn_hood_on_l(ComponentPose=@n2, Alpha=@n3, ExternalForce="0,0,0") {
    angularDampingOverride=1
    angularSpringConstant=5000
    bAngularSpring=true
    bOverrideAngularDamping=true
    bUseGravityOverride=true
    gravityScale=0.35
    simulationSpace=RootRelative
    alphaBoolBlend {blendOption=Linear}
    boundBone {boneName=dyn_hood_on_l}
    componentAppliedLinearAccClamp=<2500 2500 1500>
    componentLinearAccScale=<1 1 1>
    componentLinearVelScale=<1 1 1>
    gravityOverride {z=-980}
    physicsBodyDefinitions [{
      boundBone {boneName=dyn_hood_on_l}
      boxExtents=<35 35 35>
      constraintSetup {angularLimitsMax=<15 0 15>, angularLimitsMin=<-15 -30 -15>, angularTarget=<1 0 0>}
      localJointOffset=<4 0 0>
     }]
   }
   n5 = Anim Dynamics - Bone: dyn_hood_on_r(ComponentPose=@n4, Alpha=@n3, ExternalForce="0,0,0") {
    angularDampingOverride=1
    angularSpringConstant=5000
    bAngularSpring=true
    bOverrideAngularDamping=true
    bUseGravityOverride=true
    gravityScale=0.35
    simulationSpace=RootRelative
    alphaBoolBlend {blendOption=Linear}
    boundBone {boneName=dyn_hood_on_r}
    componentAppliedLinearAccClamp=<2500 2500 1500>
    componentLinearAccScale=<1 1 1>
    componentLinearVelScale=<1 1 1>
    gravityOverride {z=-980}
    physicsBodyDefinitions [{
      boundBone {boneName=dyn_hood_on_r}
      boxExtents=<35 35 35>
      constraintSetup {angularLimitsMax=<15 0 15>, angularLimitsMin=<-15 -30 -15>, angularTarget=<1 0 0>}
      localJointOffset=<4 0 0>
     }]
   }
   n6 = Anim Dynamics - Bone: dyn_bang_1!(ComponentPose=@n5, Alpha=@n3, GravityOverride:Vector="0,0,0", ExternalForce="0,0,0") -> (Pose) {
    angularDampingOverride=0.9999
    angularSpringConstant=2400
    bAngularSpring=true
    bOverrideAngularDamping=true
    bUseGravityOverride=true
    gravityScale=0.35
    simulationSpace=RootRelative
    alphaBoolBlend {blendInTime=2, blendOption=Cubic}
    boundBone {boneName=dyn_bang_1}
    componentAppliedLinearAccClamp=<3000 3000 3000>
    componentLinearAccScale=<0.3 0.4 0.3>
    componentLinearVelScale=<0.3 0.4 0.3>
    gravityOverride {z=-980}
    physicsBodyDefinitions [{
      boundBone {boneName=dyn_bang_1}
      boxExtents=<25 25 25>
      constraintSetup {angularLimitsMax=<0 45 10>, angularLimitsMin=<0 -5 -50>, angularTarget=<1 0 0>}
      localJointOffset=<5 0 0>
     }]
    relativeSpaceBone {boneName=head}
   }
   n7 = Anim Dynamics - Bone: dyn_bang_2!(ComponentPose=@n6, Alpha=@n3, GravityOverride:Vector="0,0,0", ExternalForce="0,0,0") -> (Pose) {
    angularDampingOverride=0.9999
    angularSpringConstant=2400
    bAngularSpring=true
    bOverrideAngularDamping=true
    bUseGravityOverride=true
    gravityScale=0.35
    simulationSpace=RootRelative
    alphaBoolBlend {blendInTime=2, blendOption=Cubic}
    boundBone {boneName=dyn_bang_2}
    componentAppliedLinearAccClamp=<3000 3000 3000>
    componentLinearAccScale=<0.3 0.4 0.3>
    componentLinearVelScale=<0.3 0.4 0.3>
    gravityOverride {z=-980}
    physicsBodyDefinitions [{
      boundBone {boneName=dyn_bang_2}
      boxExtents=<25 25 25>
      constraintSetup {angularLimitsMax=<5 50 60>, angularLimitsMin=<5 -8 0>, angularTarget=<1 0 0>}
      localJointOffset=<10 0 0>
     }]
    relativeSpaceBone {boneName=head}
   }
   n8 = Anim Dynamics - Bone: dyn_bang_3!(ComponentPose=@n7, Alpha=@n3, GravityOverride:Vector="0,0,0", ExternalForce="0,0,0") -> (Pose) {
    angularDampingOverride=0.9999
    angularSpringConstant=2400
    bAngularSpring=true
    bOverrideAngularDamping=true
    bUseGravityOverride=true
    gravityScale=0.35
    simulationSpace=RootRelative
    alphaBoolBlend {blendInTime=2, blendOption=Cubic}
    boundBone {boneName=dyn_bang_3}
    componentAppliedLinearAccClamp=<3000 3000 3000>
    componentLinearAccScale=<0.3 0.4 0.3>
    componentLinearVelScale=<0.3 0.4 0.3>
    gravityOverride {z=-980}
    physicsBodyDefinitions [{
      boundBone {boneName=dyn_bang_3}
      boxExtents=<25 25 25>
      constraintSetup {angularLimitsMax=<0 30 2>, angularLimitsMin=<0 -60 -40>, angularTarget=<1 0 0>}
      localJointOffset=<3 0 0>
     }]
    relativeSpaceBone {boneName=head}
   }
   n9 = Anim Dynamics - Bone: dyn_bang_4!(ComponentPose=@n8, Alpha=@n3, GravityOverride:Vector="0,0,0", ExternalForce="0,0,0") -> (Pose) {
    angularDampingOverride=0.9999
    angularSpringConstant=2400
    bAngularSpring=true
    bOverrideAngularDamping=true
    bUseGravityOverride=true
    gravityScale=0.35
    simulationSpace=RootRelative
    alphaBoolBlend {blendInTime=2, blendOption=Cubic}
    boundBone {boneName=dyn_bang_4}
    componentAppliedLinearAccClamp=<3000 3000 3000>
    componentLinearAccScale=<0.3 0.4 0.3>
    componentLinearVelScale=<0.3 0.4 0.3>
    gravityOverride {z=-980}
    physicsBodyDefinitions [{
      boundBone {boneName=dyn_bang_4}
      boxExtents=<25 25 25>
      constraintSetup {angularLimitsMax=<0 30 2>, angularLimitsMin=<0 -60 -40>, angularTarget=<1 0 0>}
      localJointOffset=<3 0 0>
     }]
    relativeSpaceBone {boneName=head}
   }
   n10 = RigidBody!(
    ComponentPose=@n9
    Alpha
    OverrideWorldGravity:Vector=Get GravityOverride()
    ExternalForce
    SimSpaceSettings:Sim Space Settings Structure=Get Sim Space Settings() -> (Sim Space Settings:Sim Space Settings Structure)
   ) -> (Pose) {
    bClampLinearTranslationLimitToRefPose=true
    bFreezeIncomingPoseOnStart=true
    bOverrideWorldGravity=true
    bUseExternalClothCollision=true
    overridePhysicsAsset=/BRCosmetics/Characters/Player/Female/Medium/Bodies/F_MED_PlatypusBranch/Meshes/Parts/F_MED_PlatypusBranch_FaceAcc_Physics
    simulationSpace=BaseBoneSpace
    alphaBoolBlend {blendOption=Linear}
    baseBoneRef {boneName=spine_04}
    componentAppliedLinearAccClamp=<3000 3000 3000>
    overrideWorldGravity {z=-980}
    simSpaceSettings {maxAngularAcceleration=100, maxAngularVelocity=4, maxLinearAcceleration=700, maxLinearVelocity=300, velocityScaleZ=0.5, worldAlpha=1}
   }
   n11 = "Transform (Modify) Bone - Bone: base_hood_on_r"(
    ComponentPose=@n10
    Alpha=Get Is Targeting Alpha() -> (Is Targeting Alpha:double)
    Translation="1.467,6.000,-1.885"
    Rotation="0,0,0"
    Scale="1,1,1"
   ) {
    rotationMode=BMM_Additive
    rotationSpace=BCS_BoneSpace
    translationMode=BMM_Additive
    translationSpace=BCS_BoneSpace
    alphaBoolBlend {blendOption=Linear}
    alphaScaleBiasClamp {bInterpResult=true}
    boneToModify {boneName=base_hood_on_r}
    translation=<1.467 6 -1.885>
   }
   n12 = "Transform (Modify) Bone - Bone: base_bang_2"(ComponentPose=@n11, Alpha=Get headUp() -> (headUp:double), Translation="0,1,-1", Rotation="0,0,0", Scale="1,1,1") {
    rotationMode=BMM_Additive
    rotationSpace=BCS_BoneSpace
    translationMode=BMM_Additive
    translationSpace=BCS_BoneSpace
    alphaBoolBlend {blendOption=Linear}
    boneToModify {boneName=base_bang_2}
    translation {y=1, z=-1}
   }
   n13 = Get constraintsPonytail() -> (constraintsPonytail:double)
   n14 = "Constraint - base_braid (PPPPPPPP)"!(
    ComponentPose=@n12
    Alpha=@n13
    ConstraintWeights_0=0.3783
    ConstraintWeights_1:float=0.2634
    ConstraintWeights_2:float=0.1884
    ConstraintWeights_3:float=0.1032
    ConstraintWeights_4:float=0.034
    ConstraintWeights_5:float=0.0183
    ConstraintWeights_6:float=0.0143
    ConstraintWeights_7:float=0.0001
   ) -> (Pose) {
    alphaBoolBlend {blendInTime=0.2, blendOption=Linear, blendOutTime=0.2}
    boneToModify {boneName=base_braid}
    constraintSetup [
     {targetBone {boneName=spine_05}}
     {targetBone {boneName=neck_01}}
     {targetBone {boneName=neck_02}}
     {targetBone {boneName=head}}
     {targetBone {boneName=clavicle_r}}
     {targetBone {boneName=spine_04}}
     {targetBone {boneName=spine_03}}
     {targetBone {boneName=clavicle_l}}
    ]
    constraintWeights [1,1,1,1,1,1,1,1]
   }
   n15 = "Constraint - dyn_braid_00 (PPPPPPPP)"!(
    ComponentPose=@n14
    Alpha=@n13
    ConstraintWeights_0=0.3783
    ConstraintWeights_1:float=0.2634
    ConstraintWeights_2:float=0.1884
    ConstraintWeights_3:float=0.1032
    ConstraintWeights_4:float=0.034
    ConstraintWeights_5:float=0.0183
    ConstraintWeights_6:float=0.0143
    ConstraintWeights_7:float=0.0001
   ) -> (Pose) {
    alphaBoolBlend {blendInTime=0.2, blendOption=Linear, blendOutTime=0.2}
    boneToModify {boneName=dyn_braid_00}
    constraintSetup [
     {targetBone {boneName=spine_05}}
     {targetBone {boneName=neck_01}}
     {targetBone {boneName=neck_02}}
     {targetBone {boneName=head}}
     {targetBone {boneName=clavicle_r}}
     {targetBone {boneName=spine_04}}
     {targetBone {boneName=spine_03}}
     {targetBone {boneName=clavicle_l}}
    ]
    constraintWeights [1,1,1,1,1,1,1,1]
   }
   n16 = "Constraint - dyn_braid_01 (PPPPPPP)"!(
    ComponentPose=@n15
    Alpha=@n13
    ConstraintWeights_0=0.4829
    ConstraintWeights_1:float=0.2082
    ConstraintWeights_2:float=0.1212
    ConstraintWeights_3:float=0.1073
    ConstraintWeights_4:float=0.0441
    ConstraintWeights_5:float=0.027
    ConstraintWeights_6:float=0.0093
   ) -> (Pose) {
    alphaBoolBlend {blendInTime=0.2, blendOption=Linear, blendOutTime=0.2}
    boneToModify {boneName=dyn_braid_01}
    constraintSetup [
     {targetBone {boneName=spine_05}}
     {targetBone {boneName=neck_01}}
     {targetBone {boneName=neck_02}}
     {targetBone {boneName=clavicle_r}}
     {targetBone {boneName=head}}
     {targetBone {boneName=spine_04}}
     {targetBone {boneName=spine_03}}
    ]
    constraintWeights [1,1,1,1,1,1,1]
   }
   n17 = "Constraint - dyn_braid_02 (PPPPPPP)"!(
    ComponentPose=@n16
    Alpha=@n13
    ConstraintWeights_0=0.5898
    ConstraintWeights_1:float=0.2394
    ConstraintWeights_2:float=0.0749
    ConstraintWeights_3:float=0.0568
    ConstraintWeights_4:float=0.0276
    ConstraintWeights_5:float=0.0089
    ConstraintWeights_6:float=0.0026
   ) -> (Pose) {
    alphaBoolBlend {blendInTime=0.2, blendOption=Linear, blendOutTime=0.2}
    boneToModify {boneName=dyn_braid_02}
    constraintSetup [
     {targetBone {boneName=spine_05}}
     {targetBone {boneName=clavicle_r}}
     {targetBone {boneName=spine_04}}
     {targetBone {boneName=neck_01}}
     {targetBone {boneName=neck_02}}
     {targetBone {boneName=spine_03}}
     {targetBone {boneName=head}}
    ]
    constraintWeights [1,1,1,1,1,1,1]
   }
   n18 = "Constraint - dyn_braid_03 (PPPPPP)"!(
    ComponentPose=@n17
    Alpha=@n13
    ConstraintWeights_0=0.7127
    ConstraintWeights_1:float=0.1234
    ConstraintWeights_2:float=0.1163
    ConstraintWeights_3:float=0.028
    ConstraintWeights_4:float=0.0109
    ConstraintWeights_5:float=0.0087
   ) -> (Pose) {
    alphaBoolBlend {blendInTime=0.2, blendOption=Linear, blendOutTime=0.2}
    boneToModify {boneName=dyn_braid_03}
    constraintSetup [
     {targetBone {boneName=spine_05}}
     {targetBone {boneName=clavicle_r}}
     {targetBone {boneName=spine_04}}
     {targetBone {boneName=neck_01}}
     {targetBone {boneName=neck_02}}
     {targetBone {boneName=spine_03}}
    ]
    constraintWeights [1,1,1,1,1,1]
   }
   Output Pose(Result=Component To Local(ComponentPose=@n18)) {layerGroup=None}
   Get bIsTargeting()
  ]}
 EventGraph {nodes [
   Event BlueprintUpdateAnimation()
   > n1 = Sequence!(execute) -> (then_0, then_1, then_2:Exec, then_3:Exec)
   > n2 = Branch(Condition=Get OnZipline() -> (OnZipline:bool))
   > Set animD alpha(animD alpha:double=0) -> (Output_Get:double)
   Set animD alpha(execute=@n2.else, animD alpha=1)
   n3 = Gravity Override State Function DEPRECATED(
    execute=@n1.then_1
    CharacterPartAnimInstance=Self-Reference()
    GameTimeInSeconds=GetGameTimeInSeconds()
    GravityOverrideParams=Get Gravity Override Params()
    RBAN Params=Get RBAN Params()
   )
   > Set GravityOverride(GravityOverride=@n3)
   > Set Sim Space Settings(Sim Space Settings:Sim Space Settings Structure=Break RBANSetting Struct!(RBANSettingStruct:"RBANSetting Struct Structure (by ref)"=@n3.RBAN SettingsOut) -> (
     ComponentLinearAccScale_28_#1:Vector=~
     ComponentLinearVelScale_31_#2:Vector=~
     SimSpaceSettings_34_#3:Sim Space Settings Structure
    ) {is=title K2Node_BreakStruct}.SimSpaceSettings_34_#3) -> (Output_Get:Sim Space Settings Structure)
   Set Is Targeting Alpha(execute=@n1.then_2, Is Targeting Alpha:double="To Float (Boolean)"(InBool=Get bIsTargeting())) -> (Output_Get:double)
   > Set lod1(lod1:bool="Equal (Integer)"(A=Get CurrentLOD())) -> (Output_Get:bool)
   > n4 = Branch(Condition=OR Boolean!(A=Get bIsInFrontEnd(), B=Get bIsPlayingEmote(), C:bool=Get lod1() -> (lod1:bool)) -> (ReturnValue))
   > Set constraintsPonytail(constraintsPonytail:double=0.85) -> (Output_Get:double)
   Set constraintsPonytail(execute=@n4.else, constraintsPonytail=0)
   n5 = Get Skel Mesh()
   Set headUp(
    execute=@n1.then_3
    headUp:double=GetAlphaFromReferenceTransformDelta(SkeletalMeshComponent=@n5, BoneName=spine_05, BaseName=head, InputRange=(X=0,Y=90), TwistAxis="0,0,1")
   ) -> (Output_Get:double)
   > Set headLR(headLR:double="Absolute (Float)"(A=GetAlphaFromReferenceTransformDelta(SkeletalMeshComponent=@n5, BoneName=spine_05, BaseName=head, DeltaComponentType=EulerY, InputRange=(X=0,Y=90), TwistAxis="0,0,1").RawOutput)) -> (Output_Get:double)
   Event BlueprintInitializeAnimation()
   > Set Skel Mesh(Skel Mesh=GetOwnerSkeletalMeshComponent())
   n6 = Event BlueprintProcessAnimInputEvent()
   > n7 = Branch(Condition=IsNamed(self=@n6.AnimInputEvent, InEventName=aie.Player.Zipline.Entered))
   > Set OnZipline(OnZipline:bool=1) -> (Output_Get:bool)
   Branch(execute=@n7.else, Condition=IsNamed(self=@n6.AnimInputEvent, InEventName=aie.Player.Zipline.Exited))
   > Set OnZipline(OnZipline=false)
  ]}
]
variables [
 animD alpha {type=number}
 constraintsPonytail {type=number}
 Gravity Override Params {type=GravityOverrideParamsStruct}
 GravityOverride {type=Vector}
 headLR {type=number}
 headUp {type=number}
 Is Targeting Alpha {type=number}
 lod1 {type=boolean}
 OnZipline {type=boolean}
 RBAN Params {
  type=RBANParamsStruct
  default {
   dBNO {simSpaceSettings {
     maxAngularAcceleration=100
     maxAngularVelocity=4
     maxLinearAcceleration=700
     maxLinearVelocity=300
     velocityScaleZ=0.5
     externalLinearDragV=<0 0 0>
    }}
   dBNO_Carried {simSpaceSettings {
     maxAngularAcceleration=100
     maxAngularVelocity=4
     maxLinearAcceleration=700
     maxLinearVelocity=300
     velocityScaleZ=0.5
     externalLinearDragV=<0 0 0>
    }}
   emoteOrMelee {simSpaceSettings {
     maxAngularAcceleration=100
     maxAngularVelocity=4
     maxLinearAcceleration=700
     maxLinearVelocity=300
     velocityScaleZ=0.5
     externalLinearDragV=<0 0 0>
    }}
   front_End {simSpaceSettings {
     maxAngularAcceleration=100
     maxAngularVelocity=4
     maxLinearAcceleration=700
     maxLinearVelocity=300
     velocityScaleZ=0.5
     externalLinearDragV=<0 0 0>
    }}
   onGround_Crouch_Moving {simSpaceSettings {
     maxAngularAcceleration=100
     maxAngularVelocity=4
     maxLinearAcceleration=700
     maxLinearVelocity=300
     velocityScaleZ=0.5
     externalLinearDragV=<0 0 0>
    }}
   onGround_Falling {simSpaceSettings {
     maxAngularAcceleration=100
     maxAngularVelocity=4
     maxLinearAcceleration=700
     maxLinearVelocity=300
     velocityScaleZ=0.5
     externalLinearDragV=<0 0 0>
    }}
   onGround_Moving {simSpaceSettings {
     maxAngularAcceleration=100
     maxAngularVelocity=4
     maxLinearAcceleration=700
     maxLinearVelocity=300
     velocityScaleZ=0.5
     externalLinearDragV=<0 0 0>
    }}
   onGround_Standing {simSpaceSettings {
     maxAngularAcceleration=100
     maxAngularVelocity=4
     maxLinearAcceleration=700
     maxLinearVelocity=300
     velocityScaleZ=0.5
     externalLinearDragV=<0 0 0>
    }}
   rail_Grinding {simSpaceSettings {
     maxAngularAcceleration=100
     maxAngularVelocity=4
     maxLinearAcceleration=700
     maxLinearVelocity=300
     velocityScaleZ=0.5
     worldAlpha=1
     externalLinearDragV=<0 0 0>
    }}
   riding_Moving {simSpaceSettings {
     maxAngularAcceleration=100
     maxAngularVelocity=4
     maxLinearAcceleration=700
     maxLinearVelocity=300
     velocityScaleZ=0.5
     externalLinearDragV=<0 0 0>
    }}
   riding_Moving_Targeting {simSpaceSettings {
     maxAngularAcceleration=100
     maxAngularVelocity=4
     maxLinearAcceleration=700
     maxLinearVelocity=300
     velocityScaleZ=0.5
     externalLinearDragV=<0 0 0>
    }}
   riding_Standing {simSpaceSettings {
     maxAngularAcceleration=100
     maxAngularVelocity=4
     maxLinearAcceleration=700
     maxLinearVelocity=300
     velocityScaleZ=0.5
     externalLinearDragV=<0 0 0>
    }}
   skydive_DivingDown {simSpaceSettings {
     maxAngularAcceleration=100
     maxAngularVelocity=4
     maxLinearAcceleration=700
     maxLinearVelocity=300
     velocityScaleZ=0.5
     externalLinearDragV=<0 0 0>
    }}
   skydive_FloatingDown {
    componentLinearAccScale=<0 0.5 0>
    componentLinearVelScale=<0 0.5 0>
    simSpaceSettings {
     maxAngularAcceleration=100
     maxAngularVelocity=4
     maxLinearAcceleration=700
     maxLinearVelocity=300
     velocityScaleZ=0.5
     externalLinearDragV=<0 0 0>
    }
   }
   skydive_GoingUp {simSpaceSettings {
     maxAngularAcceleration=100
     maxAngularVelocity=4
     maxLinearAcceleration=700
     maxLinearVelocity=300
     velocityScaleZ=0.5
     externalLinearDragV=<0 0 0>
    }}
   skydive_Parachuting {
    componentLinearAccScale=<0 0.1 0.3>
    componentLinearVelScale=<0 0.1 0.3>
    simSpaceSettings {
     maxAngularAcceleration=100
     maxAngularVelocity=4
     maxLinearAcceleration=700
     maxLinearVelocity=300
     velocityScaleZ=0.5
     externalLinearDragV=<0 0 0>
    }
   }
   skydive_Surfing {simSpaceSettings {
     maxAngularAcceleration=100
     maxAngularVelocity=4
     maxLinearAcceleration=700
     maxLinearVelocity=300
     velocityScaleZ=0.5
     externalLinearDragV=<0 0 0>
    }}
   swimming {simSpaceSettings {
     maxAngularAcceleration=100
     maxAngularVelocity=4
     maxLinearAcceleration=700
     maxLinearVelocity=300
     velocityScaleZ=0.5
     externalLinearDragV=<0 0 0>
    }}
   tactical_Sprint {simSpaceSettings {
     maxAngularAcceleration=100
     maxAngularVelocity=4
     maxLinearAcceleration=700
     maxLinearVelocity=300
     velocityScaleZ=0.5
     externalLinearDragV=<0 0 0>
    }}
   ziplining {simSpaceSettings {
     maxAngularAcceleration=100
     maxAngularVelocity=4
     maxLinearAcceleration=700
     maxLinearVelocity=300
     velocityScaleZ=0.5
     externalLinearDragV=<0 0 0>
    }}
  }
 }
 Sim Space Settings {
  type=SimSpaceSettings
  default {
   dampingAlpha=1
   maxAngularAcceleration=10000
   maxAngularVelocity=10000
   maxLinearAcceleration=10000
   maxLinearVelocity=10000
   velocityScaleZ=1
   worldAlpha=0
   externalLinearDragV=<0 0 0>
  }
 }
 Skel Mesh {type=SkeletalMeshComponent}
]
